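; Tank turret weapon Blueprint that adds cosmetic firing and overheat feedback for the tank's mounted gun.
; On equip it caches the host tank and attaches the overheat audio component to its PassengerGun socket.
; Muzzle flash events shake the local player's camera and spawn effects at the left or right muzzle by OffHand.
; Overheat changes fade the loop in or out and drive its interpolated heat parameter, while overheating plays a one-shot.
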(event EventBeginPlay
  (|CallParentFunction)
  (Overheating|BindEventtoOnOverheatValueChanged (EventDispatchers|CreateEvent :!event EventOnOVerheatChange)))

(event FortWeapon|EventOnEquippedAndAttached
  ;[1 1 1] Handle Camera Shake on Fire and Overheat Audio
  (|CallParentFunction)
  (Networking|CanExecuteCosmeticEvents
    (:True
      (bind _astank_vehicle (Utilities|Casting|CastToTank_Vehicle (Utilities|ToObject(Interface) (Vehicle|GetHostVehicle))))
      (Variables|Default|SetTankVehicle _astank_vehicle)
      (CallFunction|AttachAudio))))

(event EventOnPlayWeaponFireFX
  (params (bPersistentFire :type Boolean) (bSecondaryFire :type Boolean)))

(event FortWeapon|EventOnUnEquippedAndDetached
  ;[1 1 1] Handle Camera Shake on Fire and Overheat Audio
  (|CallParentFunction)
  (Networking|CanExecuteCosmeticEvents
    (:True
      (CallFunction|StopOverheatAudio))))

(event EventMuzzleFlashFX
  (params (Persistent%20Fire :type Boolean) (OffHand :type Boolean))
  (bind _tankvehicle (Variables|Default|GetTankVehicle))
  (bind _returnvalue (Class|FortVehicleInterface|GetSkeletalMeshComponent _tankvehicle))
  (bind _returnvalue_2 (Game|GetInstigator))
  (bind _asfort_player_pawn (Utilities|Casting|CastToFortPlayerPawn _returnvalue_2))
  (bind _returnvalue_3 (Pawn|GetLocalViewingPlayerController _asfort_player_pawn))
  (bind _fxmuzzleflashns (Variables|Default|GetFXMuzzleFlashNS))
  ;[1 1 1] Handle Camera Shake on Fire and Overheat Audio
  (Networking|CanExecuteCosmeticEvents
    (:True
      (Utilities|IsValid _returnvalue_3
        (:Is%20Valid
          (CameraShakes|StartCameraShake :self (Variables|/Script/Engine.PlayerController|GetPlayerCameraManager _returnvalue_3) :ShakeClass (Variables|Default|GetTurretFireCameraShake) :Scale 1.0 :PlaySpace "CameraLocal"))
        (:Is%20Not%20Valid
          _))
      (Utilities|IsValid (Variables|Default|GetTankVehicle)
        (:Is%20Valid
          (if OffHand
            (Effects|SpawnFXSystemAttached _fxmuzzleflashns _returnvalue "muzzleRight") ; Right
            (else
              (bind _returnvalue_4 (Effects|SpawnFXSystemAttached _fxmuzzleflashns _returnvalue "muzzleLeft")) ; Left
              (Effects|Components|ParticleSystem|SetFloatParameter _returnvalue_4 "User.MuzzleLeft" 1.0))))))))

(event Overheating|EventOnStartOverheated
  ;[1 1 1] Handle Camera Shake on Fire and Overheat Audio
  (|CallParentFunction)
  (Networking|CanExecuteCosmeticEvents
    (:True
      (Audio|SpawnSoundAttached (Variables|Sound|GetOverheatedSound) (Variables|Default|GetAudioOverheat))
      (CallFunction|StopOverheatAudio))))

(event Overheating|EventOnFinishedCooling
  ;[1 1 1] Handle Camera Shake on Fire and Overheat Audio
  (|CallParentFunction)
  (CallFunction|StopOverheatAudio))

(event EventEndPlay
  (params (EndPlayReason :type EEndPlayReason%20Enum :default "Destroyed"))
  ;[1 1 1] Handle Camera Shake on Fire and Overheat Audio
  (|CallParentFunction EndPlayReason)
  (CallFunction|StopOverheatAudio))

(event Custom|EventOnOverheatChange
  (params (CurrentValue :type Float%20%28single-precision%29) (MaxValue :type Float%20%28single-precision%29) (State :type EFortWeaponOverheatState%20Enum) (bIsInCooldown :type Boolean))
  ;[1 1 1] Handle Camera Shake on Fire and Overheat Audio
  ;;[1 1 1] Update overheat audio
  (switch Utilities|FlowControl|Switch|SwitchonEFortWeaponOverheatState State
    (:None
      (CallFunction|StopOverheatAudio))
    (:Heating
      (Utilities|FlowControl|Sequence
        (:then_0
          (bind _audiooverheat (Variables|Default|GetAudioOverheat)
            (:then
              (if (not (Audio|Components|Audio|IsPlaying _audiooverheat))
                (Audio|Components|Audio|FadeIn _audiooverheat 0.25 1.0 0.0 "Linear" :!class /Script/Engine.AudioComponent)))
            (:else)))
        (:then_1
          _)))
    (:Cooling
      _))
  (bind _output_get (Variables|Sound|SetOverheatAudioParam (Math|Interpolation|FInterpTo (Variables|Sound|GetOverheatAudioParam) (Math|Float|SelectFloat 0.0 (/ CurrentValue MaxValue) bIsInCooldown) (Utilities|Time|GetWorldDeltaSeconds) 1e+01)))
  (bind _audiooverheat_2 (Variables|Default|GetAudioOverheat)
    (:then
      (Audio|Parameter|SetFloatParameter _audiooverheat_2 "Overheat" _output_get :!class /Script/AudioExtensions.AudioParameterControllerInterface))
    (:else)))

(fn AttachAudio
  (bind _tankvehicle (Variables|Default|GetTankVehicle)
    (:then
      (bind _audiooverheat (Variables|Default|GetAudioOverheat)
        (:then
          (Transformation|AttachComponentToComponent _audiooverheat (Class|FortVehicleInterface|GetSkeletalMeshComponent _tankvehicle) "PassengerGun" "SnapToTarget" "SnapToTarget" "SnapToTarget" true)
          (Transformation|AddLocalOffset _audiooverheat "0, 0,50.000000" false false))
        (:else)))
    (:else)))

(fn HideReticle)

(fn ShowReticle)

(fn StopOverheatAudio
  (bind _audiooverheat (Variables|Default|GetAudioOverheat)
    (:then
      (if (Audio|Components|Audio|IsPlaying _audiooverheat)
        (Audio|Components|Audio|FadeOut _audiooverheat 1.0 0.0 "Linear" :!class /Script/Engine.AudioComponent)))
    (:else)))

(fn ConstructionScript
  (|CallParentFunction))
